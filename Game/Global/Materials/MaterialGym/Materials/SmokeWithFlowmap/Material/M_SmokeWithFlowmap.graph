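# M_SmokeWithFlowmap is an unlit material in the MaterialGym folder, named for a flowmap-driven smoke effect.
# Its graph holds a single call to MF_SmokeWithFlowMap, so the effect's logic lives in that material function rather than in this material.
# The function's inputs cover a tech art mask, emissive and flow speed.
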Material /Game/Global/Materials/MaterialGym/Materials/SmokeWithFlowmap/Material/M_SmokeWithFlowmap
graph {
 properties {pixelDepthOffsetMode=PDOM_Legacy, shadingModel=MSM_Unlit, usedShadingModels=Unlit}
 nodes [MaterialFunctionCall!(TechArtMaskInput, EmissiveInput, FlowSpeed) {materialFunction=/Game/Global/Materials/MaterialGym/Materials/SmokeWithFlowmap/MaterialFunction/MF_SmokeWithFlowMap}]
}
